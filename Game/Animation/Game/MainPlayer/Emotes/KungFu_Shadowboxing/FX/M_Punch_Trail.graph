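# Translucent, unlit, two-sided mesh-particle material for the punch trail in the KungFu Shadowboxing emote.
# Axis-swapped, panning smoke and noise layers, including a distorted, partly posterized one, form an end-faded shape that drives opacity and optional two-color tinting of particle color.
# An optional mask slides via the Panner offset dynamic parameter, with near-camera fade and non-mobile depth fade.
# Switchable SplineThicken WPO sets width, with optional particle-driven scaling and noise displacement.
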Material /Game/Animation/Game/MainPlayer/Emotes/KungFu_Shadowboxing/FX/M_Punch_Trail
graph {
 properties {
  bUsedWithMeshParticles=true
  bUsedWithNiagaraMeshParticles=true
  blendMode=BLEND_Translucent
  pixelDepthOffsetMode=PDOM_Legacy
  shadingModel=MSM_Unlit
  twoSided=true
  usedShadingModels=Unlit
 }
 nodes [
  n1 = ParticleColor()
  n2 = Constant() {r=1}
  n3 = TextureCoordinate() {vTiling=0.5}
  n4 = Multiply(A=ComponentMask(None=@n3) {g=true}, B=ScalarParameter() {defaultValue=1, parameterName=Texture_UVScale})
  n5 = ComponentMask(None=@n3) {r=true}
  n6 = ScalarParameter() {defaultValue=-0.3, parameterName=Texture_A_Panner_speed}
  n7 = Constant()
  n8 = TextureSample(UVs=Panner(
    Coordinate=MaterialFunctionCall(X=@n4, Y=@n5) {materialFunction=&MakeFloat2}.Result
    Speed=AppendVector(A=@n6, B=@n7)
   ) {speedX=-0.3}) {
   samplerSource=SSM_Wrap_WorldGroupSettings
   samplerType=SAMPLERTYPE_Grayscale
   texture=&T_Smoke_Trail_Soft
  }
  n9 = Power(Base=@n8.R)
  n10 = TextureCoordinate()
  n11 = ComponentMask(None=@n10) {g=true}
  n12 = ComponentMask(None=@n10) {r=true}
  n13 = AppendVector(A=ScalarParameter() {defaultValue=0.5, parameterName=Texture_B_Panner_speed}, B=Constant())
  n14 = TextureSample(UVs=Panner(
    Coordinate=MaterialFunctionCall(X=@n11, Y=@n12) {materialFunction=&MakeFloat2}.Result
    Speed=@n13
   ) {speedX=0.5}) {samplerType=SAMPLERTYPE_Grayscale, texture=&LowResBlurredNoise}
  n15 = TextureCoordinate()
  n16 = Multiply(A=ComponentMask(None=@n15) {g=true}, B=ScalarParameter() {defaultValue=1, parameterName=Texture_UVScale})
  n17 = ComponentMask(None=@n15) {r=true}
  n18 = TextureSample(UVs=Panner(
    Coordinate=MaterialFunctionCall(X=@n16, Y=@n17) {materialFunction=&MakeFloat2}.Result
    Speed=@n13
   ) {speedX=0.5}) {
   samplerSource=SSM_Wrap_WorldGroupSettings
   samplerType=SAMPLERTYPE_Grayscale
   texture=&T_Smoke_Trail_Soft
  }
  n19 = Multiply(A=LinearInterpolate(A=Multiply(A=@n9) {constB=2}, B=Multiply(A=@n14.R)), B=@n18.R)
  n20 = Multiply(A=@n19) {constB=2}
  n21 = TextureCoordinate()
  n22 = MaterialFunctionCall(
   X=Multiply(A=ComponentMask(None=@n21) {g=true}, B=ScalarParameter() {defaultValue=1, parameterName=Texture_UVScale})
   Y=ComponentMask(None=@n21) {r=true}
  ) {materialFunction=&MakeFloat2}
  n23 = Panner(Coordinate=TextureCoordinate() {vTiling=2}) {speedY=-0.5}
  n24 = Multiply(A=ComponentMask(None=TextureSample(UVs=@n23) {samplerType=SAMPLERTYPE_LinearColor, texture=&T_NoisePack_LinearColor}.RGB) {g=true, r=true}) {constB=0.07}
  n25 = VectorParameter() {parameterName=Base Tex Speed, defaultValue=<rgba -0.2 0 0 0>}
  n26 = TextureSample(UVs=Panner(Coordinate=Add(A=@n22.Result, B=@n24), Speed=ComponentMask(None=@n25.RGB) {g=true, r=true}) {speedX=-0.2}) {samplerType=SAMPLERTYPE_Grayscale, texture=&T_Smoke_Trail_Soft}
  n27 = Multiply(A=@n26.R, B=@n26.R)
  n28 = Multiply(A=Multiply(A=@n27, B=@n27)) {constB=2}
  n29 = Multiply(A=Round(None=Multiply(A=@n28) {constB=5})) {constB=0.2}
  n30 = Multiply(A=@n28)
  n31 = Multiply(A=Power(Base=Multiply(A=@n19, B=@n19))) {constB=50}
  n32 = Power(Base=Add(A=Add(A=@n29, B=@n30), B=LinearInterpolate(A=@n31, B=@n20) {constAlpha=0.4}))
  n33 = ScalarParameter() {parameterName=Use_sun_flare_texture}
  n34 = OneMinus(None=ComponentMask(None=TextureCoordinate()) {g=true})
  n35 = TextureCoordinate()
  n36 = Multiply(A=Multiply(A=LinearInterpolate(A=@n20, B=@n32, Alpha=@n33), B=Power(Base=@n34)), B=Power(Base=ComponentMask(None=@n35) {g=true}))
  n37 = LinearInterpolate(
   A=VectorParameter() {parameterName=Lerp_Color2, defaultValue=<rgba 1 0.0976 0 1>}.RGB
   B=VectorParameter() {parameterName=Lerp_Color1, defaultValue=<rgba 1 0.6604 0.1499 1>}.RGB
   Alpha=@n36
  )
  n38 = ScalarParameter() {parameterName=Use_Lerp_Color}
  MaterialFunctionCall(Emissive=Multiply(A=@n1.RGB, B=LinearInterpolate(A=@n2, B=@n37, Alpha=@n38))) {materialFunction=&MF_FX_EmissiveAdjustments}
  n39 = TextureCoordinate()
  n40 = ScalarParameter() {defaultValue=1, parameterName=Mask_UVscale}
  n41 = Constant()
  n42 = DynamicParameter() {defaultValue=<rgba 0 1 1 1>, paramNames [Dissolve, Panner offset, WPO, Width_Scale]}
  n43 = TextureSampleParameter2D(UVs=Panner(Coordinate=Multiply(A=@n39, B=@n40), Time=Constant() {r=1}, Speed=AppendVector(A=@n41, B=@n42.Panner offset))) {parameterName=Trail Mask, samplerSource=SSM_Clamp_WorldGroupSettings, texture=/Game/Effects/Fort_Effects/Textures/Smoke/T_Smoke_Trail_Soft_2}
  n44 = Multiply(A=@n1.A, B=StaticSwitchParameter(True=Multiply(A=@n36, B=@n43.B), False=@n36) {defaultValue=true, parameterName=Use Highlight Trail})
  n45 = ScalarParameter() {defaultValue=70, parameterName=Depth_Fade}
  Multiply(
   A=ShadingPathSwitch(Default=DepthFade(Opacity=@n44, FadeDistance=@n45), Mobile=@n44) {desc=Migrated from FeatureLevelSwitch}
   B=MaterialFunctionCall(Distance=ScalarParameter() {parameterName=NearCameraFadeDist}) {materialFunction=&Near_Camera_Fade}.Result
  )
  n46 = ScalarParameter() {defaultValue=1, parameterName=GlobalWidthScale}
  n47 = ScalarParameter() {defaultValue=30, parameterName=Widthbase}
  n48 = DynamicParameter() {defaultValue=<rgba 0 1 1 1>, paramNames [Dissolve, Panner offset, WPO, Width_Scale]}
  n49 = Multiply(A=@n47, B=@n48.Width_Scale)
  n50 = ScalarParameter() {parameterName=Use_WPO_Width_Scale}
  n51 = ScalarParameter() {defaultValue=30, parameterName=WidthTip}
  n52 = Multiply(A=@n51, B=@n48.Width_Scale)
  n53 = MaterialFunctionCall(
   WidthBase=Multiply(A=@n46, B=LinearInterpolate(A=@n47, B=@n49, Alpha=@n50))
   WidthTip=Multiply(A=@n46, B=LinearInterpolate(A=@n51, B=@n52, Alpha=@n50))
  ) {materialFunction=&SplineThicken}
  n54 = Panner(Coordinate=TextureCoordinate()) {speedY=0.5}
  n55 = Multiply(
   A=Transform(None=TextureSample(UVs=@n54) {samplerType=SAMPLERTYPE_LinearColor, texture=&T_NoisePack_LinearColor}.RGB)
   B=DynamicParameter() {defaultValue=<rgba 0 1 1 1>, paramNames [Dissolve, Panner offset, WPO, Width_Scale]}.WPO
  )
  StaticSwitchParameter(
   True=LinearInterpolate(A=@n53.WorldPositionOffset, B=Add(A=@n53.WorldPositionOffset, B=@n55), Alpha=ScalarParameter() {parameterName=Add_WPO_Noise})
   False=Constant()
  ) {defaultValue=true, parameterName=Use SplineThicken}
 ]
}
